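# Parent cloth-shading material for the BirthMitten quadruped companion, layering a flipbook face and animated eyes onto the 2023 anime shader.
# Face colors blend into lit and shaded inputs before eye animation, with a companion normal-sphere override feeding shading.
# Brow animation curves offset flipbook UVs, mirrored by pre-skinned local X.
# Helmet or collar VO activity gates emissive boost, dimmed for frontend helmets; a static switch adds Kicks sock support.
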Material /CosmeticCompanions/Assets/Quadruped/BirthMitten/Materials/M_BirthMitten_Parent
graph {
 expressions {
  MaterialExpressionAdd_1 {
   class=MaterialExpressionAdd
   inputs [
    A {from=MaterialExpressionTextureCoordinate_1}
    B {from=MaterialExpressionComponentMask_3}
   ]
  }
  MaterialExpressionAppendVector_0 {
   class=MaterialExpressionAppendVector
   inputs [
    A {from=MaterialExpressionVertexColor_0, output=""}
    B {from=MaterialExpressionVertexColor_0, output=""}
   ]
  }
  MaterialExpressionChannelMaskParameter_0 {
   class=MaterialExpressionChannelMaskParameter
   inputs [None {from=MaterialExpressionStaticSwitchParameter_1, output=""}]
   properties {group=01_EyeAnimation_Parameters, parameterName=EyeMaskChannel, sortPriority=-32}
  }
  MaterialExpressionComponentMask_0 {class=MaterialExpressionComponentMask, inputs [None {from=MaterialExpressionTextureCoordinate_0}], properties {r=true}}
  MaterialExpressionComponentMask_3 {class=MaterialExpressionComponentMask, inputs [None {from=MaterialExpressionLinearInterpolate_3}], properties {g=true, r=true}}
  MaterialExpressionComponentMask_6 {class=MaterialExpressionComponentMask, inputs [None {from=MaterialExpressionVertexInterpolator_1, output=PS}], properties {r=true}}
  MaterialExpressionConstant_0 {class=MaterialExpressionConstant}
  MaterialExpressionConstant_1 {class=MaterialExpressionConstant}
  MaterialExpressionConstant_2 {class=MaterialExpressionConstant, properties {r=0.01}}
  MaterialExpressionConvert_3 {
   class=MaterialExpressionConvert
   inputs [None {from=MaterialExpressionComponentMask_3}]
   properties {
    nodeName=Break Float2
    convertInputs [{
      type=Vector2
      expressionInput {
       expression=&self:MaterialExpressionComponentMask_3
       inputName=None
       mask=0
       maskA=0
       maskB=0
       maskG=0
       maskR=0
       outputIndex=0
      }
     }]
    convertOutputs [{type=Scalar}, {type=Scalar}]
   }
  }
  MaterialExpressionConvert_4 {
   class=MaterialExpressionConvert
   inputs [
    None {from=MaterialExpressionConvert_3}
    None {from=MaterialExpressionMultiply_3}
   ]
   properties {
    nodeName=Make Vector2
    convertInputs [
     {
      type=Scalar
      expressionInput {
       expression=&self:MaterialExpressionConvert_3
       inputName=None
       mask=0
       maskA=0
       maskB=0
       maskG=0
       maskR=0
       outputIndex=0
      }
     }
     {
      type=Scalar
      expressionInput {
       expression=&self:MaterialExpressionMultiply_3
       inputName=None
       mask=0
       maskA=0
       maskB=0
       maskG=0
       maskR=0
       outputIndex=0
      }
     }
    ]
    convertOutputs [{type=Vector2}]
   }
  }
  MaterialExpressionIf_0 {
   class=MaterialExpressionIf
   inputs [
    A {from=MaterialExpressionMax_1, output=""}
    B {from="", output=""}
    "A > B" {from=MaterialExpressionLinearInterpolate_2, output=""}
    "A == B" {from=MaterialExpressionLinearInterpolate_2, output=""}
    "A < B" {from=MaterialExpressionTextureCoordinate_1, output=""}
   ]
   properties {constB=0.1}
  }
  MaterialExpressionIf_1 {
   class=MaterialExpressionIf
   inputs [
    A {from=MaterialExpressionLinearInterpolate_5, output=""}
    B {from=MaterialExpressionConstant_2, output=""}
    "A > B" {from=MaterialExpressionLinearInterpolate_6, output=""}
    "A == B" {from=MaterialExpressionLinearInterpolate_6, output=""}
    "A < B" {from=MaterialExpressionConstant_0, output=""}
   ]
   properties {}
  }
  MaterialExpressionLinearInterpolate_0 {
   class=MaterialExpressionLinearInterpolate
   inputs [
    A {from=MaterialExpressionMaterialFunctionCall_21, output=Color_Lit}
    B {from=MaterialExpressionMaterialFunctionCall_1, output=FaceColor}
    Alpha {from=MaterialExpressionNamedRerouteDeclaration_0}
   ]
  }
  MaterialExpressionLinearInterpolate_1 {
   class=MaterialExpressionLinearInterpolate
   inputs [
    A {from=MaterialExpressionMaterialFunctionCall_21, output=Color_Shaded}
    B {from=MaterialExpressionMultiply_0}
    Alpha {from=MaterialExpressionNamedRerouteDeclaration_0}
   ]
  }
  MaterialExpressionLinearInterpolate_2 {
   class=MaterialExpressionLinearInterpolate
   inputs [
    A {from=MaterialExpressionSaturate_1}
    B {from=MaterialExpressionSaturate_2}
    Alpha {from=MaterialExpressionStep_1}
   ]
  }
  MaterialExpressionLinearInterpolate_3 {
   class=MaterialExpressionLinearInterpolate
   inputs [
    A {from=MaterialExpressionVectorParameter_0, output=RGB}
    B {from=MaterialExpressionLinearInterpolate_4}
    Alpha {from=MaterialExpressionScalarParameter_1}
   ]
  }
  MaterialExpressionLinearInterpolate_4 {
   class=MaterialExpressionLinearInterpolate
   inputs [
    A {from=MaterialExpressionVectorParameter_2, output=RGB}
    B {from=MaterialExpressionVectorParameter_1, output=RGB}
    Alpha {from=MaterialExpressionScalarParameter_4}
   ]
  }
  MaterialExpressionLinearInterpolate_5 {
   class=MaterialExpressionLinearInterpolate
   inputs [
    A {from=MaterialExpressionScalarParameter_2}
    B {from=MaterialExpressionScalarParameter_6}
    Alpha {from=MaterialExpressionScalarParameter_5}
   ]
  }
  MaterialExpressionLinearInterpolate_6 {
   class=MaterialExpressionLinearInterpolate
   inputs [
    A {from=MaterialExpressionLinearInterpolate_7}
    B {from=MaterialExpressionScalarParameter_3}
    Alpha {from=MaterialExpressionScalarParameter_8}
   ]
  }
  MaterialExpressionLinearInterpolate_7 {
   class=MaterialExpressionLinearInterpolate
   inputs [
    A {from=MaterialExpressionScalarParameter_3}
    B {from=MaterialExpressionMultiply_5}
    Alpha {from=MaterialExpressionScalarParameter_9}
   ]
  }
  MaterialExpressionLocalPosition_1 {class=MaterialExpressionLocalPosition, inputs [], properties {localOrigin=InstancePreSkinning}}
  MaterialExpressionMaterialFunctionCall_0 {
   class=MaterialExpressionMaterialFunctionCall
   inputs [
    Color Lit {from=MaterialExpressionLinearInterpolate_0}
    Color Shaded {from=MaterialExpressionLinearInterpolate_1}
    Use UV2 {from=MaterialExpressionStaticBoolParameter_0}
    Mask {from=MaterialExpressionStaticSwitchParameter_0}
   ]
   properties {materialFunction=/Game/Global/Materials/Character_Parent_2023/Anime_Parent_2023/MaterialFunctions/MF_AnimeEyeAnimation_v2}
  }
  MaterialExpressionMaterialFunctionCall_1 {
   class=MaterialExpressionMaterialFunctionCall
   inputs [
    Material {from=""}
    Brow UV {from=MaterialExpressionIf_0}
    Eye UV {from=""}
    Mouth UV {from=""}
   ]
   properties {materialFunction=/CosmeticCompanions/Materials/MaterialFunctions/MF_FlipbookFace_Companion}
  }
  MaterialExpressionMaterialFunctionCall_14 {
   class=MaterialExpressionMaterialFunctionCall
   inputs [
    Shading_Gradient {from=MaterialExpressionMaterialFunctionCall_20, output=Shading_Gradient}
    Threshold_Gradient {from=MaterialExpressionMaterialFunctionCall_20, output=Shading_Gradient}
    Specular_Gradient {from=MaterialExpressionMaterialFunctionCall_20, output=Shading_Gradient}
    Sheen_Mask {from=MaterialExpressionMaterialFunctionCall_20, output=Shading_Gradient}
    Thickness {from=MaterialExpressionMaterialFunctionCall_20, output=Shading_Gradient}
    DFL_Mask {from=MaterialExpressionMaterialFunctionCall_20, output=Shading_Gradient}
    FX {from=MaterialExpressionMaterialFunctionCall_20, output=Shading_Gradient}
    Metallic_Gradient {from=MaterialExpressionMaterialFunctionCall_20, output=Shading_Gradient}
    RimLight_Gradient {from=MaterialExpressionMaterialFunctionCall_20, output=Shading_Gradient}
    NormalBuffer {from=MaterialExpressionMaterialFunctionCall_20, output=Shading_Gradient}
    -------- {from=""}
    CINE_LightSources {from=MaterialExpressionMaterialFunctionCall_20, output=Shading_Gradient}
    CINE_LightColor {from=MaterialExpressionMaterialFunctionCall_20, output=Shading_Gradient}
    --------- {from=""}
    Color_Lit {from=MaterialExpressionMaterialFunctionCall_0, output=Lit}
    Color_Shaded {from=MaterialExpressionMaterialFunctionCall_0, output=Lit}
    SRM {from=MaterialExpressionMaterialFunctionCall_21, output=SRM}
    Emissive {from=MaterialExpressionMultiply_1}
    DFL_Color {from=MaterialExpressionMaterialFunctionCall_21, output=SRM}
    Specular_Color {from=MaterialExpressionMaterialFunctionCall_21, output=SRM}
    RimLight_Color {from=MaterialExpressionMaterialFunctionCall_21, output=SRM}
   ]
   properties {materialFunction=/Game/Global/Materials/Character_Parent_2023/Anime_Parent_2023/MaterialFunctions/MFC_AnimeShader_2023_Post}
  }
  MaterialExpressionMaterialFunctionCall_2 {
   class=MaterialExpressionMaterialFunctionCall
   inputs [
    InOpacity {from=""}
    Material {from=MaterialExpressionMaterialFunctionCall_14, output=Result}
    Mask {from=""}
   ]
   properties {materialFunction=/Game/Global/Materials/Cosmetics/Kicks/Socks/Materials/MF_Sock}
  }
  MaterialExpressionMaterialFunctionCall_20 {
   class=MaterialExpressionMaterialFunctionCall
   inputs [
    CustomLightVector {from=""}
    OverrideNormal_WS {from=MaterialExpressionMaterialFunctionCall_3, output=Normal}
    STT {from=MaterialExpressionMaterialFunctionCall_21, output=STT}
    Normal_Map {from=MaterialExpressionMaterialFunctionCall_21, output=STT}
    DFL {from=MaterialExpressionMaterialFunctionCall_21, output=STT}
    FX {from=MaterialExpressionMaterialFunctionCall_21, output=STT}
   ]
   properties {materialFunction=/Game/Global/Materials/Character_Parent_2023/Anime_Parent_2023/MaterialFunctions/MFC_AnimeShader_2023_Calculations}
  }
  MaterialExpressionMaterialFunctionCall_21 {
   class=MaterialExpressionMaterialFunctionCall
   inputs [CustomUV {from=""}]
   properties {materialFunction=/Game/Global/Materials/Character_Parent_2023/Anime_Parent_2023/MaterialFunctions/MFC_AnimeShader_2023_Pre}
  }
  MaterialExpressionMaterialFunctionCall_3 {
   class=MaterialExpressionMaterialFunctionCall
   inputs [Normal {from=MaterialExpressionMaterialFunctionCall_21, output=Normal}]
   properties {materialFunction=/CosmeticCompanions/Materials/MaterialFunctions/MF_CompanionAnime_NormalSphere}
  }
  MaterialExpressionMax_0 {
   class=MaterialExpressionMax
   inputs [
    A {from=MaterialExpressionScalarParameter_36}
    B {from=MaterialExpressionScalarParameter_41}
   ]
  }
  MaterialExpressionMax_1 {
   class=MaterialExpressionMax
   inputs [
    A {from=MaterialExpressionMax_0}
    B {from=MaterialExpressionScalarParameter_38}
   ]
  }
  MaterialExpressionMultiply_0 {
   class=MaterialExpressionMultiply
   inputs [
    A {from=MaterialExpressionMaterialFunctionCall_1, output=FaceColor}
    B {from=MaterialExpressionScalarParameter_0}
   ]
  }
  MaterialExpressionMultiply_1 {
   class=MaterialExpressionMultiply
   inputs [
    A {from=MaterialExpressionMaterialFunctionCall_21, output=Emissive}
    B {from=MaterialExpressionIf_1}
   ]
  }
  MaterialExpressionMultiply_3 {
   class=MaterialExpressionMultiply
   inputs [
    A {from=MaterialExpressionConvert_3}
    B {from=""}
   ]
   properties {constB=-1}
  }
  MaterialExpressionMultiply_5 {
   class=MaterialExpressionMultiply
   inputs [
    A {from=MaterialExpressionScalarParameter_3}
    B {from=""}
   ]
   properties {constB=0.3}
  }
  MaterialExpressionNamedRerouteDeclaration_0 {
   class=MaterialExpressionNamedRerouteDeclaration
   inputs [None {from=MaterialExpressionMaterialFunctionCall_1, output=FaceAlpha}]
   properties {name=FlipbookFaceAlpha, nodeColor=<rgba 0.8281 0 1 1>}
  }
  MaterialExpressionSaturate_0 {class=MaterialExpressionSaturate, inputs [None {from=MaterialExpressionSign_0}]}
  MaterialExpressionSaturate_1 {class=MaterialExpressionSaturate, inputs [None {from=MaterialExpressionAdd_1}]}
  MaterialExpressionSaturate_2 {class=MaterialExpressionSaturate, inputs [None {from=MaterialExpressionSubtract_3}]}
  MaterialExpressionScalarParameter_0 {class=MaterialExpressionScalarParameter, properties {defaultValue=1, parameterName=FlipbookFaceAnimeShadedValue}}
  MaterialExpressionScalarParameter_1 {class=MaterialExpressionScalarParameter, properties {group="Eyes Curve Parameters (Driven By Animation)", parameterName=R_brow_Ot_Dn_pose}}
  MaterialExpressionScalarParameter_2 {class=MaterialExpressionScalarParameter, properties {group=BirthMitten, parameterName=VOActivity - Helmet}}
  MaterialExpressionScalarParameter_3 {class=MaterialExpressionScalarParameter, properties {group=BirthMitten, parameterName=Light Boost}}
  MaterialExpressionScalarParameter_36 {class=MaterialExpressionScalarParameter, properties {group="Eyes Curve Parameters (Driven By Animation)", parameterName=R_brow_Ot_Dn_pose}}
  MaterialExpressionScalarParameter_38 {class=MaterialExpressionScalarParameter, properties {group="Eyes Curve Parameters (Driven By Animation)", parameterName=R_brow_In_Dn_pose}}
  MaterialExpressionScalarParameter_4 {class=MaterialExpressionScalarParameter, properties {group="Eyes Curve Parameters (Driven By Animation)", parameterName=R_brow_In_Dn_pose}}
  MaterialExpressionScalarParameter_41 {class=MaterialExpressionScalarParameter, properties {group="Eyes Curve Parameters (Driven By Animation)", parameterName=R_brow_down_pose}}
  MaterialExpressionScalarParameter_5 {class=MaterialExpressionScalarParameter, properties {parameterName=IsCollarLight?}}
  MaterialExpressionScalarParameter_6 {class=MaterialExpressionScalarParameter, properties {parameterName=VOActivity - Collar}}
  MaterialExpressionScalarParameter_7 {class=MaterialExpressionScalarParameter, properties {defaultValue=0.4, parameterName=Temporal Responsiveness}}
  MaterialExpressionScalarParameter_8 {class=MaterialExpressionScalarParameter, properties {parameterName=IsCollarLight?}}
  MaterialExpressionScalarParameter_9 {class=MaterialExpressionScalarParameter, properties {parameterName=IsFrontend}}
  MaterialExpressionSign_0 {class=MaterialExpressionSign, inputs [None {from=MaterialExpressionSubtract_0, output=""}], properties {}}
  MaterialExpressionStaticBoolParameter_0 {
   class=MaterialExpressionStaticBoolParameter
   inputs []
   properties {defaultValue=true, group=01_EyeAnimation_Parameters, parameterName=Use UV2 for Eye Animation, sortPriority=-30}
  }
  MaterialExpressionStaticSwitchParameter_0 {
   class=MaterialExpressionStaticSwitchParameter
   inputs [
    True {from=MaterialExpressionChannelMaskParameter_0, output=R}
    False {from=MaterialExpressionSaturate_0, output=""}
   ]
   properties {group=01_EyeAnimation_Parameters, parameterName=UseEyeMask, sortPriority=-33}
  }
  MaterialExpressionStaticSwitchParameter_1 {
   class=MaterialExpressionStaticSwitchParameter
   inputs [
    True {from=MaterialExpressionAppendVector_0, output=""}
    False {from=MaterialExpressionTextureSampleParameter2D_0, output=RGBA}
   ]
   properties {group=01_EyeAnimation_Parameters, parameterName=UseVertexColorForEyeMask, sortPriority=-33}
  }
  MaterialExpressionStaticSwitchParameter_2 {
   class=MaterialExpressionStaticSwitchParameter
   inputs [
    True {from=MaterialExpressionMaterialFunctionCall_2, output=Result}
    False {from=MaterialExpressionMaterialFunctionCall_14, output=Result}
   ]
   properties {group=Kicks, parameterName=IsKicks}
  }
  MaterialExpressionStep_1 {
   class=MaterialExpressionStep
   inputs [
    Y {from="", output=""}
    X {from=MaterialExpressionComponentMask_6, output=""}
   ]
   properties {}
  }
  MaterialExpressionSubtract_0 {
   class=MaterialExpressionSubtract
   inputs [
    A {from=MaterialExpressionComponentMask_0}
    B {from=""}
   ]
   properties {constB=0.0001}
  }
  MaterialExpressionSubtract_3 {
   class=MaterialExpressionSubtract
   inputs [
    A {from=MaterialExpressionTextureCoordinate_1}
    B {from=MaterialExpressionConvert_4}
   ]
  }
  MaterialExpressionTemporalResponsivenessOutput_0 {class=MaterialExpressionTemporalResponsivenessOutput, inputs [None {from=MaterialExpressionScalarParameter_7, output=""}], properties {}}
  MaterialExpressionTextureCoordinate_0 {class=MaterialExpressionTextureCoordinate, properties {coordinateIndex=1}}
  MaterialExpressionTextureCoordinate_1 {class=MaterialExpressionTextureCoordinate, properties {coordinateIndex=1}}
  MaterialExpressionTextureSampleParameter2D_0 {
   class=MaterialExpressionTextureSampleParameter2D
   inputs [
    UVs {from="", output=""}
    Apply View MipBias {from="", output=""}
   ]
   properties {
    group=01_EyeAnimation_Parameters
    parameterName=EyeMask
    samplerSource=SSM_Wrap_WorldGroupSettings
    samplerType=SAMPLERTYPE_Masks
    sortPriority=-32
    texture=&T_WhiteMask
   }
  }
  MaterialExpressionVectorParameter_0 {class=MaterialExpressionVectorParameter, properties {parameterName=Brow Custom Offset, defaultValue=<rgba 0.0447 -0.0801 0 0>}}
  MaterialExpressionVectorParameter_1 {class=MaterialExpressionVectorParameter, properties {parameterName=Brow In Dn - Offset}}
  MaterialExpressionVectorParameter_2 {class=MaterialExpressionVectorParameter, properties {parameterName=Brow Ot Dn - Offset, defaultValue=<rgba 0.0278 -0.0676 0 0>}}
  MaterialExpressionVertexColor_0 {class=MaterialExpressionVertexColor, inputs [], properties {}}
  MaterialExpressionVertexInterpolator_1 {class=MaterialExpressionVertexInterpolator, inputs [VS {from=MaterialExpressionLocalPosition_1, output=XY}], properties {}}
 }
 properties {
  allowTranslucentCustomDepthWrites=true
  bAutomaticallySetUsageInEditor=false
  bHasPixelAnimation=true
  bUseMaterialAttributes=true
  bUsedWithClothing=true
  bUsedWithMeshDeformer=true
  bUsedWithMorphTargets=true
  bUsedWithNiagaraMeshParticles=true
  bUsedWithSkeletalMesh=true
  bUsedWithStaticLighting=true
  pixelDepthOffsetMode=PDOM_Legacy
  shadingModel=MSM_Cloth
  usedShadingModels=Cloth
 }
}
