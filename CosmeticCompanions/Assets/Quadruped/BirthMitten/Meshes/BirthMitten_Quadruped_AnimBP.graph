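# Animation Blueprint for the BirthMitten quadruped cosmetic companion, driving its body through Retarget Pose From Mesh with RTG_BirthMitten.
# Each update compares selected facial curves, mostly right-side, against cached values using varying tolerances.
# Any change sets the materials' Temporal Responsiveness to 1, otherwise 0, then the values are cached.
# R_blink_pose re-saves its own variable, so blink is compared to an unchanging value.
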AnimBlueprint /CosmeticCompanions/Assets/Quadruped/BirthMitten/Meshes/BirthMitten_Quadruped_AnimBP
target_skeleton=&/BirthMitten_Quadruped_Skeleton
graphs [
 AnimGraph {nodes [
   Copy Pose From Mesh() {bCopyCurves=true, bUseAttachedParent=true}
   Output Pose(Result=Retarget Pose From Mesh!() -> (Pose:Pose Link Structure) {is=title AnimGraphNode_RetargetPoseFromMesh, iKRetargeterAsset=&/RTG_BirthMitten})
  ]}
 EventGraph {nodes [
   Event BlueprintUpdateAnimation()
   n1 = GetOwningComponent()
   n2 = GetCurveValue!(self:Skeletal Mesh Component Object Reference=@n1, CurveName=R_brow_In_Dn_pose, DefaultValue:float=0) -> (Value:float, ReturnValue:bool=false)
   n3 = NOT Boolean(A="Nearly Equal (Float)"(A=@n2, B=Get R_brow_In_Dn_pose() -> (R_brow_In_Dn_pose:double), ErrorTolerance=0))
   n4 = GetCurveValue(self=@n1, CurveName=R_brow_down_pose, DefaultValue=0) -> (ReturnValue=false)
   n5 = NOT Boolean(A="Nearly Equal (Float)"(A=@n4, B=Get R_brow_down_pose() -> (R_brow_down_pose:double), ErrorTolerance=0))
   n6 = GetCurveValue(self=@n1, CurveName=R_smile_pose, DefaultValue=0) -> (ReturnValue=false)
   n7 = Get Error Tolerance() -> (Error Tolerance:double)
   n8 = NOT Boolean(A="Nearly Equal (Float)"(A=@n6, B=Get R_smile_pose() -> (R_smile_pose:double), ErrorTolerance=@n7))
   n9 = Get R_blink_pose() -> (R_blink_pose:double)
   n10 = NOT Boolean(A="Nearly Equal (Float)"(A=GetCurveValue(self=@n1, CurveName=R_blink_pose, DefaultValue=0) -> (ReturnValue=false), B=@n9, ErrorTolerance=0.1))
   n11 = GetCurveValue(self=@n1, CurveName=R_frown_pose, DefaultValue=0) -> (ReturnValue=false)
   n12 = NOT Boolean(A="Nearly Equal (Float)"(A=@n11, B=Get R_frown_pose() -> (R_frown_pose:double), ErrorTolerance=@n7))
   n13 = GetCurveValue(self=@n1, CurveName=R_lip_corner_narrow_pose, DefaultValue=0) -> (ReturnValue=false)
   n14 = NOT Boolean(A="Nearly Equal (Float)"(A=@n13, B=Get R_lip_corner_narrow_pose() -> (R_lip_corner_narrow_pose:double), ErrorTolerance=@n7))
   n15 = GetCurveValue(self=@n1, CurveName=L_smile_pose, DefaultValue=0) -> (ReturnValue=false)
   n16 = NOT Boolean(A="Nearly Equal (Float)"(A=@n15, B=Get L_smile_pose() -> (L_smile_pose:double), ErrorTolerance=0.1))
   n17 = GetCurveValue(self=@n1, CurveName=L_upper_lip_raiser_pose, DefaultValue=0) -> (ReturnValue=false)
   n18 = NOT Boolean(A="Nearly Equal (Float)"(A=@n17, B=Get L_upper_lip_raiser_pose() -> (L_upper_lip_raiser_pose:double), ErrorTolerance=@n7))
   n19 = GetCurveValue(self=@n1, CurveName=R_wide_pose, DefaultValue=0) -> (ReturnValue=false)
   n20 = NOT Boolean(A="Nearly Equal (Float)"(A=@n19, B=Get R_wide_pose() -> (R_wide_pose:double), ErrorTolerance=@n7))
   n21 = GetCurveValue(self=@n1, CurveName=jaw_open_pose, DefaultValue=0) -> (ReturnValue=false)
   n22 = NOT Boolean(A="Nearly Equal (Float)"(A=@n21, B=Get jaw_open_pose() -> (jaw_open_pose:double), ErrorTolerance=0.1))
   n23 = GetCurveValue(self=@n1, CurveName=R_extra_pose, DefaultValue=0) -> (ReturnValue=false)
   n24 = NOT Boolean(A="Nearly Equal (Float)"(A=@n23, B=Get R_extra_pose() -> (R_extra_pose:double), ErrorTolerance=@n7))
   n25 = GetCurveValue(self=@n1, CurveName=R_lower_lid_up_pose, DefaultValue=0) -> (ReturnValue=false)
   n26 = NOT Boolean(A="Nearly Equal (Float)"(A=@n25, B=Get R_lower_lid_up_pose() -> (R_lower_lid_up_pose:double), ErrorTolerance=@n7))
   n27 = GetCurveValue(self=@n1, CurveName=R_brow_squeeze_in_pose, DefaultValue=0) -> (ReturnValue=false)
   n28 = NOT Boolean(A="Nearly Equal (Float)"(A=@n27, B=Get R_brow_squeeze_in_pose() -> (R_brow_squeeze_in_pose:double), ErrorTolerance=@n7))
   n29 = GetCurveValue(self=@n1, CurveName=R_extra2_pose, DefaultValue=0) -> (ReturnValue=false)
   n30 = NOT Boolean(A="Nearly Equal (Float)"(A=@n29, B=Get R_extra2_pose() -> (R_extra2_pose:double), ErrorTolerance=@n7))
   n31 = GetCurveValue(self=@n1, CurveName=R_brow_In_Up_pose, DefaultValue=0) -> (ReturnValue=false)
   n32 = NOT Boolean(A="Nearly Equal (Float)"(A=@n31, B=Get R_brow_In_Up_pose() -> (R_brow_In_Up_pose:double), ErrorTolerance=0))
   n33 = GetCurveValue(self=@n1, CurveName=R_brow_Ot_Dn_pose, DefaultValue=0) -> (ReturnValue=false)
   n34 = NOT Boolean(A="Nearly Equal (Float)"(A=@n33, B=Get R_brow_Ot_Dn_pose() -> (R_brow_Ot_Dn_pose:double), ErrorTolerance=@n7))
   n35 = GetCurveValue(self=@n1, CurveName=R_upper_lip_raiser_pose, DefaultValue=0) -> (ReturnValue=false)
   n36 = NOT Boolean(A="Nearly Equal (Float)"(A=@n35, B=Get R_upper_lip_raiser_pose() -> (R_upper_lip_raiser_pose:double), ErrorTolerance=0))
   > SetScalarParameterValueOnMaterials!(
    execute
    self:Mesh Component Object Reference=@n1
    ParameterName:Name=Temporal Responsiveness
    ParameterValue:float=SelectFloat(
     A=1
     B=0
     bPickA=OR Boolean!(
      A=@n5
      B=@n3
      C:bool=@n32
      D:bool=@n34
      E:bool=@n28
      F:bool=@n30
      G:bool=@n24
      H:bool=@n26
      I:bool=@n20
      J:bool=@n22
      K:bool=@n16
      L:bool=@n18
      M:bool=@n12
      N:bool=@n14
      O:bool=@n8
      P:bool=@n36
      Q:bool=@n10
     ) -> (ReturnValue)
    )
   ) -> (then)
   > Set R_brow_down_pose(R_brow_down_pose:double=@n4) -> (Output_Get:double)
   > Set R_brow_In_Dn_pose(R_brow_In_Dn_pose:double=@n2) -> (Output_Get:double)
   > Set R_brow_In_Up_pose(R_brow_In_Up_pose:double=@n31) -> (Output_Get:double)
   > Set R_brow_Ot_Dn_pose(R_brow_Ot_Dn_pose:double=@n33) -> (Output_Get:double)
   > Set R_brow_squeeze_in_pose(R_brow_squeeze_in_pose:double=@n27) -> (Output_Get:double)
   > Set R_extra2_pose(R_extra2_pose:double=@n29) -> (Output_Get:double)
   > Set R_extra_pose(R_extra_pose:double=@n23) -> (Output_Get:double)
   > Set R_lower_lid_up_pose(R_lower_lid_up_pose:double=@n25) -> (Output_Get:double)
   > Set R_wide_pose(R_wide_pose:double=@n19) -> (Output_Get:double)
   > Set jaw_open_pose(jaw_open_pose:double=@n21) -> (Output_Get:double)
   > Set L_smile_pose(L_smile_pose:double=@n15) -> (Output_Get:double)
   > Set L_upper_lip_raiser_pose(L_upper_lip_raiser_pose:double=@n17) -> (Output_Get:double)
   > Set R_frown_pose(R_frown_pose:double=@n11) -> (Output_Get:double)
   > Set R_lip_corner_narrow_pose(R_lip_corner_narrow_pose:double=@n13) -> (Output_Get:double)
   > Set R_smile_pose(R_smile_pose:double=@n6) -> (Output_Get:double)
   > Set R_upper_lip_raiser_pose(R_upper_lip_raiser_pose:double=@n35) -> (Output_Get:double)
   > Set R_blink_pose(R_blink_pose:double=@n9) -> (Output_Get:double)
   TryGetPawnOwner()
  ]}
]
variables [
 Error Tolerance {type=number, default=0.25}
 jaw_open_pose {type=number}
 L_smile_pose {type=number}
 L_upper_lip_raiser_pose {type=number}
 R_blink_pose {type=number}
 R_brow_down_pose {type=number}
 R_brow_In_Dn_pose {type=number}
 R_brow_In_Up_pose {type=number}
 R_brow_Ot_Dn_pose {type=number}
 R_brow_squeeze_in_pose {type=number}
 R_extra2_pose {type=number}
 R_extra_pose {type=number}
 R_frown_pose {type=number}
 R_lip_corner_narrow_pose {type=number}
 R_lower_lid_up_pose {type=number}
 R_smile_pose {type=number}
 R_upper_lip_raiser_pose {type=number}
 R_wide_pose {type=number}
]
