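# Post-process style Anim Blueprint for the BirthMitten quadruped cosmetic companion, adding secondary physics motion to its medallion and ears.
# The input pose runs through chained single-bone Anim Dynamics nodes: a gravity-driven medallion simulated relative to root, then angular-spring ear and ear-accessory bones.
# All nodes share DynamicsAlpha with inverted scale and bias, so higher values reduce the physics.
# A PoseAsset blend node exists but is unconnected, and the EventGraph is empty.
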AnimBlueprint /CosmeticCompanions/Assets/Quadruped/BirthMitten/Meshes/BirthMitten_Quadruped_Post_AnimBP
target_skeleton=&/BirthMitten_Quadruped_Skeleton
graphs [
 AnimGraph {nodes [
   Pose!(PoseAsset:Pose Asset Object Reference=Get PoseAsset() -> (PoseAsset:Pose Asset Object Reference), SourcePose:Pose Link Structure) -> (Pose:Pose Link Structure) {is=title AnimGraphNode_PoseBlendNode, bIgnorePoseAssetSourceAnimationCurves=true, blendOption=Linear}
   n1 = Get DynamicsAlpha() -> (DynamicsAlpha:float)
   n2 = Anim Dynamics - Bone: medalion(ComponentPose=Local To Component(LocalPose=Input Pose()), Alpha=@n1, ExternalForce="0,0,0") {
    angularDampingOverride=0.6
    angularSpringConstant=1
    bOverrideAngularDamping=true
    bUseGravityOverride=true
    linearSpringConstant=1
    simulationSpace=BoneRelative
    alphaScaleBiasClamp {bias=1, scale=-1}
    boundBone {boneName=medalion}
    componentAppliedLinearAccClamp=<1500 1500 1500>
    gravityOverride {z=-900}
    physicsBodyDefinitions [{
      boundBone {boneName=medalion}
      boxExtents=<5 5 1>
      constraintSetup {coneAngle=30, angularLimitsMax=<0 0 20>, angularLimitsMin=<0 -35 -20>, angularTarget=<0 0 0>}
      localJointOffset=<3 0 0>
     }]
    relativeSpaceBone {boneName=root}
    simSpaceSettings {maxAngularAcceleration=200, maxAngularVelocity=50}
   }
   n3 = Anim Dynamics - Bone: dyn_faceAcc_ear_R(ComponentPose=@n2, Alpha=@n1, ExternalForce="0,0,0") {
    angularDampingOverride=0.9999
    angularSpringConstant=10
    bAngularSpring=true
    bOverrideAngularDamping=true
    linearSpringConstant=1500
    numSolverIterationsPostUpdate=4
    alphaScaleBiasClamp {bias=1, scale=-1}
    boundBone {boneName=dyn_faceAcc_ear_R}
    componentAppliedLinearAccClamp=<1500 1500 1500>
    gravityOverride {z=-900}
    physicsBodyDefinitions [{
      boundBone {boneName=dyn_faceAcc_ear_R}
      boxExtents=<10 10 10>
      constraintSetup {angularTargetAxis=AxisY, coneAngle=30, angularLimitsMax=<10 10 20>, angularLimitsMin=<-10 -20 -20>, angularTarget=<1 0 0>}
      localJointOffset=<1 0 0>
     }]
    relativeSpaceBone {boneName=pelvis}
    simSpaceSettings {maxAngularAcceleration=250, maxAngularVelocity=250}
   }
   n4 = Anim Dynamics - Bone: dyn_faceAcc_ear2_R(ComponentPose=@n3, Alpha=@n1, ExternalForce="0,0,0") {
    angularDampingOverride=0.9999
    angularSpringConstant=10
    bAngularSpring=true
    bOverrideAngularDamping=true
    linearSpringConstant=1500
    numSolverIterationsPostUpdate=4
    alphaScaleBiasClamp {bias=1, scale=-1}
    boundBone {boneName=dyn_faceAcc_ear2_R}
    componentAppliedLinearAccClamp=<1500 1500 1500>
    gravityOverride {z=-900}
    physicsBodyDefinitions [{
      boundBone {boneName=dyn_faceAcc_ear2_R}
      boxExtents=<10 10 10>
      constraintSetup {angularTargetAxis=AxisY, coneAngle=30, angularLimitsMax=<10 10 20>, angularLimitsMin=<-10 -20 -20>, angularTarget=<1 0 0>}
      localJointOffset=<1 0 0>
     }]
    relativeSpaceBone {boneName=pelvis}
    simSpaceSettings {maxAngularAcceleration=250, maxAngularVelocity=250}
   }
   n5 = Anim Dynamics - Bone: dyn_faceAcc_ear_L(ComponentPose=@n4, Alpha=@n1, ExternalForce="0,0,0") {
    angularDampingOverride=0.9999
    angularSpringConstant=10
    bAngularSpring=true
    bOverrideAngularDamping=true
    linearSpringConstant=1500
    numSolverIterationsPostUpdate=4
    alphaScaleBiasClamp {bias=1, scale=-1}
    boundBone {boneName=dyn_faceAcc_ear_L}
    componentAppliedLinearAccClamp=<1500 1500 1500>
    gravityOverride {z=-900}
    physicsBodyDefinitions [{
      boundBone {boneName=dyn_faceAcc_ear_L}
      boxExtents=<10 10 10>
      constraintSetup {angularTargetAxis=AxisY, coneAngle=30, angularLimitsMax=<10 10 20>, angularLimitsMin=<-10 -20 -20>, angularTarget=<1 0 0>}
      localJointOffset=<-1 0 0>
     }]
    relativeSpaceBone {boneName=pelvis}
    simSpaceSettings {maxAngularAcceleration=250, maxAngularVelocity=250}
   }
   n6 = Anim Dynamics - Bone: dyn_faceAcc_ear2_L(ComponentPose=@n5, Alpha=@n1, ExternalForce="0,0,0") {
    angularDampingOverride=0.9999
    angularSpringConstant=10
    bAngularSpring=true
    bOverrideAngularDamping=true
    linearSpringConstant=1500
    numSolverIterationsPostUpdate=4
    alphaScaleBiasClamp {bias=1, scale=-1}
    boundBone {boneName=dyn_faceAcc_ear2_L}
    componentAppliedLinearAccClamp=<1500 1500 1500>
    gravityOverride {z=-900}
    physicsBodyDefinitions [{
      boundBone {boneName=dyn_faceAcc_ear2_L}
      boxExtents=<10 10 10>
      constraintSetup {angularTargetAxis=AxisY, coneAngle=30, angularLimitsMax=<10 10 20>, angularLimitsMin=<-10 -20 -20>, angularTarget=<1 0 0>}
      localJointOffset=<-1 0 0>
     }]
    relativeSpaceBone {boneName=pelvis}
    simSpaceSettings {maxAngularAcceleration=250, maxAngularVelocity=250}
   }
   n7 = Anim Dynamics - Bone: ear_02_l(ComponentPose=@n6, Alpha=@n1, ExternalForce="0,0,0") {
    angularDampingOverride=0.9999
    angularSpringConstant=10
    bAngularSpring=true
    bOverrideAngularDamping=true
    linearSpringConstant=1500
    numSolverIterationsPostUpdate=4
    alphaScaleBiasClamp {bias=1, scale=-1}
    boundBone {boneName=ear_02_l}
    componentAppliedLinearAccClamp=<1500 1500 1500>
    gravityOverride {z=-900}
    physicsBodyDefinitions [{
      boundBone {boneName=ear_02_l}
      boxExtents=<10 10 10>
      constraintSetup {angularTargetAxis=AxisY, coneAngle=30, angularLimitsMax=<0 15 15>, angularLimitsMin=<0 0 -15>, angularTarget=<1 0 0>}
      localJointOffset=<1 0 0>
     }]
    relativeSpaceBone {boneName=pelvis}
    simSpaceSettings {maxAngularAcceleration=250, maxAngularVelocity=250}
   }
   n8 = Anim Dynamics - Bone: ear_03_l(ComponentPose=@n7, Alpha=@n1, ExternalForce="0,0,0") {
    angularDampingOverride=0.9999
    angularSpringConstant=10
    bAngularSpring=true
    bOverrideAngularDamping=true
    linearSpringConstant=1500
    numSolverIterationsPostUpdate=4
    alphaScaleBiasClamp {bias=1, scale=-1}
    boundBone {boneName=ear_03_l}
    componentAppliedLinearAccClamp=<1500 1500 1500>
    gravityOverride {z=-900}
    physicsBodyDefinitions [{
      boundBone {boneName=ear_03_l}
      boxExtents=<10 10 10>
      constraintSetup {angularTargetAxis=AxisY, coneAngle=30, angularLimitsMax=<0 15 15>, angularLimitsMin=<0 0 -15>, angularTarget=<1 0 0>}
      localJointOffset=<1 0 0>
     }]
    relativeSpaceBone {boneName=pelvis}
    simSpaceSettings {maxAngularAcceleration=250, maxAngularVelocity=250}
   }
   n9 = Anim Dynamics - Bone: ear_04_l(ComponentPose=@n8, Alpha=@n1, ExternalForce="0,0,0") {
    angularDampingOverride=0.9999
    angularSpringConstant=10
    bAngularSpring=true
    bOverrideAngularDamping=true
    linearSpringConstant=1500
    numSolverIterationsPostUpdate=4
    alphaScaleBiasClamp {bias=1, scale=-1}
    boundBone {boneName=ear_04_l}
    componentAppliedLinearAccClamp=<1500 1500 1500>
    gravityOverride {z=-900}
    physicsBodyDefinitions [{
      boundBone {boneName=ear_04_l}
      boxExtents=<10 10 10>
      constraintSetup {angularTargetAxis=AxisY, coneAngle=30, angularLimitsMax=<0 15 15>, angularLimitsMin=<0 0 -15>, angularTarget=<1 0 0>}
      localJointOffset=<1 0 0>
     }]
    relativeSpaceBone {boneName=pelvis}
    simSpaceSettings {maxAngularAcceleration=250, maxAngularVelocity=250}
   }
   n10 = Anim Dynamics - Bone: ear_02_r(ComponentPose=@n9, Alpha=@n1, ExternalForce="0,0,0") {
    angularDampingOverride=0.9999
    angularSpringConstant=10
    bAngularSpring=true
    bOverrideAngularDamping=true
    linearSpringConstant=1500
    numSolverIterationsPostUpdate=4
    alphaScaleBiasClamp {bias=1, scale=-1}
    boundBone {boneName=ear_02_r}
    componentAppliedLinearAccClamp=<1500 1500 1500>
    gravityOverride {z=-900}
    physicsBodyDefinitions [{
      boundBone {boneName=ear_02_r}
      boxExtents=<10 10 10>
      constraintSetup {angularTargetAxis=AxisY, coneAngle=30, angularLimitsMax=<0 15 15>, angularLimitsMin=<0 0 -15>, angularTarget=<1 0 0>}
      localJointOffset=<-1 0 0>
     }]
    relativeSpaceBone {boneName=pelvis}
    simSpaceSettings {maxAngularAcceleration=250, maxAngularVelocity=250}
   }
   n11 = Anim Dynamics - Bone: ear_03_r(ComponentPose=@n10, Alpha=@n1, ExternalForce="0,0,0") {
    angularDampingOverride=0.9999
    angularSpringConstant=10
    bAngularSpring=true
    bOverrideAngularDamping=true
    linearSpringConstant=1500
    numSolverIterationsPostUpdate=4
    alphaScaleBiasClamp {bias=1, scale=-1}
    boundBone {boneName=ear_03_r}
    componentAppliedLinearAccClamp=<1500 1500 1500>
    gravityOverride {z=-900}
    physicsBodyDefinitions [{
      boundBone {boneName=ear_03_r}
      boxExtents=<10 10 10>
      constraintSetup {angularTargetAxis=AxisY, coneAngle=30, angularLimitsMax=<0 15 15>, angularLimitsMin=<0 0 -15>, angularTarget=<1 0 0>}
      localJointOffset=<-1 0 0>
     }]
    relativeSpaceBone {boneName=pelvis}
    simSpaceSettings {maxAngularAcceleration=250, maxAngularVelocity=250}
   }
   n12 = Anim Dynamics - Bone: ear_04_r(ComponentPose=@n11, Alpha=@n1, ExternalForce="0,0,0") {
    angularDampingOverride=0.9999
    angularSpringConstant=10
    bAngularSpring=true
    bOverrideAngularDamping=true
    linearSpringConstant=1500
    numSolverIterationsPostUpdate=4
    alphaScaleBiasClamp {bias=1, scale=-1}
    boundBone {boneName=ear_04_r}
    componentAppliedLinearAccClamp=<1500 1500 1500>
    gravityOverride {z=-900}
    physicsBodyDefinitions [{
      boundBone {boneName=ear_04_r}
      boxExtents=<10 10 10>
      constraintSetup {angularTargetAxis=AxisY, coneAngle=30, angularLimitsMax=<0 15 15>, angularLimitsMin=<0 0 -15>, angularTarget=<1 0 0>}
      localJointOffset=<-1 0 0>
     }]
    relativeSpaceBone {boneName=pelvis}
    simSpaceSettings {maxAngularAcceleration=250, maxAngularVelocity=250}
   }
   Output Pose(Result=Component To Local(ComponentPose=@n12))
  ]}
 EventGraph {links [], nodes []}
]
variables [PoseAsset {type=PoseAsset, default=/CosmeticCompanions/Automation_Template/companion_pose_asset_template}]
